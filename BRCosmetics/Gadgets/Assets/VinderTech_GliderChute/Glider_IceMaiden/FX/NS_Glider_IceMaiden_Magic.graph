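# Niagara magic and snow effect for the Ice Maiden glider, adding snowflakes, clouds, sparkles and ribbon trails around the glider mesh.
# Most emitters attach to glider sockets or bones, with continuous spawning gated by bIsFullyDeployed.
# ForwardDot drives front-cloud rate, track-snow velocity and ribbon material animation.
# GPU Head particles follow random multi-point splines from system arrays and seed Magic and Snowflakes via particle reads, as Horn does for Horn_Magic.
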NiagaraSystem /BRCosmetics/Gadgets/Assets/VinderTech_GliderChute/Glider_IceMaiden/FX/NS_Glider_IceMaiden_Magic
dependencies {
 usedDataInterfaces [
  &NiagaraDataInterfaceParticleRead
  &NiagaraDataInterfaceCurve
  &NiagaraDataInterfaceColorCurve
  &NiagaraDataInterfaceArrayFloat3
 ]
 usedDynamicInputs [&SetBoolByFloatComparison]
 usedModules [
  &self:SystemSpawnScript.SystemScriptSource.SystemScriptGraph.NiagaraNodeAssignment_1.SetVariables_#1
  &SystemState
  &Timeline
 ]
}
emitters {
 Big_Snow {
  emitterUpdateScript [
   &EmitterState(Life Cycle Mode=::NewEnumerator1, Loop Behavior=::NewEnumerator1, Loop Duration Mode=::NewEnumerator1, UseLoopDelay=false)
   &SpawnBurst_Instantaneous(Use Spawn Probability=false, UseLoopCountLimit=false)
   &SpawnBurst_Instantaneous(Spawn Group=1, Use Spawn Probability=false, UseLoopCountLimit=false) {module=SpawnBurst_Instantaneous001}
  ]
  particleSpawnScript [
   &InitializeParticle(
    Color=&MaskLinearColorBySpawnGroup(Masked Linear Color=<rgba 1 1 1 0.15>, Passthrough Linear Color=<rgba 1 1 1 1>)
    Lifetime=5
    Sprite Rotation Angle=&MaskFloatBySpawnGroup(Masked Float=0.085, Passthrough Float=0)
    Sprite Rotation Mode=::NewEnumerator1
    Sprite Size Mode=::NewEnumerator3
    Uniform Sprite Size=&MaskFloatBySpawnGroup(Masked Float=40, Passthrough Float=20)
    UsePositionOffset=false
    Write Material Random=true
   )
   &SkeletalMeshLocation(
    Bone / Socket Index=1
    MaskSpawnGroup=false
    Mesh Sampling Type=::NewEnumerator5
    Skeletal Mesh=$User.Glider
    Socket Sampling Mode=::NewEnumerator5
   )
   &CameraOffset(Camera Offset Amount=3)
   &SpriteFacingAndAlignment(Sprite Facing=<0 1 0>, WriteAlignment=false, WriteFacing=true)
  ]
  particleUpdateScript [
   &ParticleState(Kill Particles When Lifetime Has Elapsed=false, Let Infinitely Lived Particles Die When Emitter Deactivates=true)
   &DynamicMaterialParameters(
    Index 0 Param 1=&FloatFromCurve(FloatCurve=DataInterface() {propertyValues {Curve {keys [<at 0 1>,<at 1 -0.25>]}}}, CurveIndex=$System.Timeline.NormalizedTime)
    Index 0 Param 4=&MaskFloatBySpawnGroup(Masked Float=2, Passthrough Float=1)
    Module.Param0WriteEnabled=true
    Module.Param1WriteEnabled=true
    Module.Param2WriteEnabled=true
    Module.Param3WriteEnabled=true
   )
  ]
  rendererClasses [&NiagaraSpriteRendererProperties]
  renderers [{rendererClass=&NiagaraSpriteRendererProperties}]
 }
 Front_Snow_Clouds {
  emitterUpdateScript [
   &EmitterState()
   &SpawnRate(
    SpawnRate=&MakeCustomFloatFromBool(
     Bool=$User.bIsFullyDeployed
     True Float=&Multiply_Float(
      A=20
      B=&FloatFromCurve(FloatCurve=DataInterface() {propertyValues {Curve {keys [<at 0 0 RCIM_Constant>,<at 0.2 1 RCIM_Constant>]}}}, CurveIndex=$User.ForwardDot)
     )
     False Float=0
    )
    Use Spawn Probability=false
   )
  ]
  particleSpawnScript [
   &InitializeParticle(
    Color=<rgba 1 1 1 0.075>
    Lifetime Max=1
    Lifetime Min=0.35
    Lifetime Mode=::NewEnumerator1
    Sprite Rotation Mode=::NewEnumerator2
    Sprite Size Max=<200 240>
    Sprite Size Min=<100 100>
    Sprite Size Mode=::NewEnumerator2
    UsePositionOffset=false
    Write Material Random=true
   )
   &AddVelocity(Velocity=<0 250 0>)
  ]
  particleUpdateScript [
   &ParticleState()
   &SkeletalMeshLocation(
    Bone / Socket Index=&RandomRangeInt(Minimum=0, Maximum=1)
    Bone Sampling Mode=::NewEnumerator4
    MaskSpawnGroup=false
    Position Sampling=::NewEnumerator1
    Skeletal Mesh=$User.Glider
    Source Movement Scale:Vector3f=<-0.125 -0.125 0>
   )
   &SubUVAnimation()
   &ScaleColor(
    Scale Alpha=&FloatFromCurve(
     FloatCurve=DataInterface() {propertyValues {Curve {keys [<at 0 0 RCIM_Cubic>,<at 0.2 1 RCIM_Cubic>,<at 0.75 0 RCIM_Cubic>]}}}
     Scale Curve=$System.Timeline001.NormalizedTime
    )
    Scale RGB=&MakeVectorFromLinearColorRGB(Color=&ColorFromCurve(ColorCurve=DataInterface() {propertyValues {
        AlphaCurve {keys []}
        BlueCurve {keys [<at 0 1 RCIM_Cubic>,<at 1 1 RCIM_Cubic>]}
        GreenCurve {keys [<at 0 0.7238 RCIM_Cubic>,<at 1 0.8711 RCIM_Cubic>]}
        RedCurve {keys [<at 0 0.6317 RCIM_Cubic>,<at 1 0.7999 RCIM_Cubic>]}
       }}))
    ScaleA=true
    ScaleRGB=true
   )
   &ScaleSpriteSize(Uniform Curve Sprite Scale=DataInterface() {propertyValues {Curve {keys [{tangentMode=RCTM_User, time=0, value=0.25}, {tangentMode=RCTM_User, time=1, value=1}]}}})
   &Drag(Drag=2, Use Linear Drag=true, Use Rotational Drag=false)
   &SolveForcesAndVelocity(Clamp Velocity=false, Limit Acceleration=false)
  ]
  rendererClasses [&NiagaraSpriteRendererProperties]
  renderers [{rendererClass=&NiagaraSpriteRendererProperties}]
 }
 Head {
  simTarget=GPUComputeSim
  emitterUpdateScript [
   &EmitterState()
   &SpawnRate(
    SpawnRate=&MakeCustomFloatFromBool(Bool=$User.bIsFullyDeployed, True Float=5, False Float=0)
    Use Spawn Probability=false
   )
  ]
  particleSpawnScript [
   &InitializeParticle(Color Mode=::NewEnumerator0, Sprite Size Mode=::NewEnumerator3, Uniform Sprite Size=2, UsePositionOffset=false, Write Material Random=true)
   /SetVariables(Particles.RandomSplineChoice:NiagaraInt32=&RandomRangeInt(Minimum=0, Maximum=16)) {isSetParameters=true}
  ]
  particleUpdateScript [
   &ParticleState()
   &self:SplineThroughMultiplePoints(
    FifthPosition:Vector3f=&SelectVectorFromArray(Direct Array Index=$Particles.RandomSplineChoice, Vector Selection Array=$System.Spline 5th Positions)
    FirstPosition:Vector3f=&SelectVectorFromArray(Direct Array Index=$Particles.RandomSplineChoice, Vector Selection Array=$System.Spline 1st Positions)
    FourthPosition:Vector3f=&SelectVectorFromArray(Direct Array Index=$Particles.RandomSplineChoice, Vector Selection Array=$System.Spline 4th Positions)
    SecondPosition:Vector3f=&SelectVectorFromArray(Direct Array Index=$Particles.RandomSplineChoice, Vector Selection Array=$System.Spline 2nd Positions)
    SixthPosition:Vector3f=&SelectVectorFromArray(Direct Array Index=$Particles.RandomSplineChoice, Vector Selection Array=$System.Spline 6th Positions)
    T:NiagaraFloat=$Particles.NormalizedAge
    ThirdPosition:Vector3f=&SelectVectorFromArray(Direct Array Index=$Particles.RandomSplineChoice, Vector Selection Array=$System.Spline 3rd Positions)
   ) {module=SplineThroughMultiplePoints}
  ]
  rendererClasses [&NiagaraSpriteRendererProperties]
  renderers [
   {rendererClass=&NiagaraSpriteRendererProperties}
   {rendererClass=&NiagaraSpriteRendererProperties, rendererIndex=1}
   {rendererClass=&NiagaraSpriteRendererProperties, rendererIndex=2}
   {rendererClass=&NiagaraSpriteRendererProperties, rendererIndex=3}
   {rendererClass=&NiagaraSpriteRendererProperties, rendererIndex=4}
   {rendererClass=&NiagaraSpriteRendererProperties, rendererIndex=5}
   {rendererClass=&NiagaraSpriteRendererProperties, rendererIndex=6}
  ]
 }
 Horn {
  simTarget=GPUComputeSim
  emitterUpdateScript [
   &EmitterState()
   &SpawnRate(
    SpawnRate=&MakeCustomFloatFromBool(Bool=$User.bIsFullyDeployed, True Float=10, False Float=0)
    Use Spawn Probability=false
   )
  ]
  particleSpawnScript [
   &InitializeParticle(
    Color=<rgba 1 1 1 1.5>
    Lifetime Max=1.25
    Lifetime Min=0.85
    Lifetime Mode=::NewEnumerator1
    Sprite Rotation Mode=::NewEnumerator2
    Sprite Size Mode=::NewEnumerator1
    Uniform Sprite Size Max=7
    Uniform Sprite Size Min=1
    UsePositionOffset=false
    Write Material Random=true
   )
   &SampleSkeletalMesh(
    AdditionalYawPitchRoll=false
    Mesh Sampling Type:ENiagaraSkelSamplingModeFull=::NewEnumerator5
    Sampled Position Offset:Vector3f=<-5 0 0>
    Skeletal Mesh:NiagaraDataInterfaceSkeletalMesh=$User.Glider
    Socket Sampling Mode:ENiagaraSocketSamplingMode=::NewEnumerator5
   ) {enabled=false}
   &SkeletalMeshLocation(
    MaskSpawnGroup=false
    Mesh Sampling Type=::NewEnumerator5
    Orientation Sampling=::NewEnumerator1
    Skeletal Mesh=$User.Glider
    Socket Sampling Mode=::NewEnumerator5
   )
   /SetVariables(Particles.Position=$Emitter.SampleSkeletalMesh.SampledPosition) {enabled=false, isSetParameters=true}
   &ShapeLocation(
    Cone Angle=3
    MaskSpawnGroup=false
    Override Seed=false
    Rotation Axis=<-1 0 0>
    Rotation Mode=::NewEnumerator1
    Shape Origin=$Particles.SkeletalMeshLocation.SampledPosition
    Shape Primitive=::NewEnumerator5
    UseOwnerScale=true
   ) {enabled=false, module=ShapeLocation001}
   &ShapeLocation(
    Cone Angle=1
    Cone Inner Angle=2
    MaskSpawnGroup=false
    Offset Mode=::NewEnumerator0
    Override Seed=false
    Rotation Mode=::NewEnumerator2
    Shape Origin=$Particles.SkeletalMeshLocation.SampledPosition
    Shape Primitive=::NewEnumerator5
    UseOwnerScale=true
    Yaw / Pitch / Roll=<0 0 -0.13>
   )
   &AddVelocity(Velocity=&Multiply_VectorByFloat(Vector=$Particles.ShapeLocation.ShapeNormal, Float=&RandomRangeFloat(Minimum=2, Maximum=5)))
   &AddVelocity(Velocity=&RotateVectorByQuaternion(
     Vector To Rotate=&MakeVector(X=&RandomRangeFloat(Minimum=5, Maximum=15), Y=0, Z=0)
     Quaternion=$Emitter.SampleSkeletalMesh.SampledRotation
    )) {enabled=false, module=AddVelocity001}
   &CameraOffset(Camera Offset Amount=5)
  ]
  particleUpdateScript [
   &ParticleState()
   &AccelerationForce(Acceleration=&MakeVector(X=0, Y=0, Z=&RandomRangeFloat(Minimum=5, Maximum=20)))
   &Drag(Drag=&RandomRangeFloat(Minimum=1, Maximum=3), Use Linear Drag=true, Use Rotational Drag=false)
   &SolveForcesAndVelocity(Clamp Velocity=false, Limit Acceleration=false)
   &SpriteRotationRate(Rotation Rate=&RandomRangeFloat(Minimum=-0.1, Maximum=0.12))
   &ScaleColor(
    Scale Alpha=&FloatFromCurve(FloatCurve=DataInterface() {propertyValues {Curve {keys [
         {interpMode=RCIM_Cubic, leaveTangentWeight=0.3333, tangentMode=RCTM_User, time=0, value=1}
         {arriveTangentWeight=0.3333, interpMode=RCIM_Cubic, tangentMode=RCTM_User, time=1, value=0}
        ]}}})
    ScaleA=true
    ScaleRGB=true
   )
   &DynamicMaterialParameters(
    Index 0 Param 1=&FloatFromCurve(FloatCurve=DataInterface() {propertyValues {Curve {keys [
         {interpMode=RCIM_Cubic, leaveTangentWeight=0.15, tangentMode=RCTM_User, time=0, value=1.2}
         {
          arriveTangent=-2.0203
          arriveTangentWeight=0.3808
          interpMode=RCIM_Cubic
          leaveTangent=-2.0203
          leaveTangentWeight=0.1269
          tangentMode=RCTM_User
          time=0.45
          value=0
         }
         {arriveTangentWeight=0.1038, interpMode=RCIM_Cubic, leaveTangentWeight=0.2767, tangentMode=RCTM_User, time=0.6, value=-0.2}
         {arriveTangentWeight=0.1333, interpMode=RCIM_Cubic, tangentMode=RCTM_User, time=1, value=1}
        ]}}})
    Index 0 Param 2=&FloatFromCurve(FloatCurve=DataInterface() {propertyValues {Curve {keys [
         {interpMode=RCIM_Cubic, leaveTangentWeight=0.2833, tangentMode=RCTM_User, time=0, value=3.5}
         {arriveTangentWeight=1.0313, interpMode=RCIM_Cubic, leaveTangentWeight=0.182, tangentMode=RCTM_User, time=0.85, value=2.8}
         {arriveTangentWeight=0.05, interpMode=RCIM_Cubic, tangentMode=RCTM_User, time=1, value=0}
        ]}}})
    Index 0 Param 3=&FloatFromCurve(FloatCurve=DataInterface() {propertyValues {Curve {keys [
         {interpMode=RCIM_Cubic, leaveTangentWeight=0.1, tangentMode=RCTM_User, time=0.2, value=0}
         {arriveTangentWeight=0.1, interpMode=RCIM_Cubic, leaveTangentWeight=0.1267, tangentMode=RCTM_User, time=0.5, value=4}
         {arriveTangentWeight=0.1267, interpMode=RCIM_Cubic, tangentMode=RCTM_User, time=0.88, value=0}
        ]}}})
    Index 0 Param 4=&RandomRangeFloat(Minimum=0, Maximum=4)
    Module.Param0WriteEnabled=true
    Module.Param1WriteEnabled=true
    Module.Param2WriteEnabled=true
    Module.Param3WriteEnabled=true
   )
   &ScaleSpriteSize(Uniform Curve Sprite Scale=DataInterface() {propertyValues {Curve {keys [<at 0 1 RCIM_Cubic>,<at 0.75 1.2 RCIM_Cubic>,<at 1 1 RCIM_Cubic>]}}})
  ]
  rendererClasses [&NiagaraSpriteRendererProperties]
  renderers [{rendererClass=&NiagaraSpriteRendererProperties}]
 }
 Horn_Magic {
  simTarget=GPUComputeSim
  emitterUpdateScript [
   &EmitterState()
   &SpawnParticlesFromOtherEmitter(
    Attribute Reader=DataInterface() {propertyValues {EmitterBinding {bindingMode=Other, emitterName=Horn}}}
    Spawn Rate=3
    Spawn Rate Per Particle Cap=500
    Use Maximum Age=false
    Use Minimum Age=false
   )
   &SpawnParticlesFromOtherEmitter(
    Attribute Reader=DataInterface() {propertyValues {EmitterBinding {bindingMode=Other, emitterName=Horn}}}
    Spawn Group=1
    Spawn Rate=5
    Spawn Rate Per Particle Cap=500
    Use Maximum Age=false
    Use Minimum Age=false
   ) {module=SpawnParticlesFromOtherEmitter001}
  ]
  particleSpawnScript [
   &InitializeParticle(
    Color=&MaskLinearColorBySpawnGroup(Masked Linear Color=<rgba 0.5729 0.8173 1 1>, Passthrough Linear Color=<rgba 0 0.5722 1 1>)
    Lifetime Max=0.5
    Lifetime Min=0.125
    Lifetime Mode=::NewEnumerator1
    Sprite Size Mode=::NewEnumerator1
    Uniform Sprite Size Max=8
    Uniform Sprite Size Min=3
    UsePositionOffset=false
    Write Material Random=true
   )
   &SampleParticlesFromOtherEmitter()
   &SphereLocation(
    MaskSpawnGroup=false
    Override Seed=false
    Sphere Radius=&MaskFloatBySpawnGroup(Masked Float=0, Passthrough Float=0.1)
    Surface Only=false
    UseNonUniformScale=false
    UseOffset=false
    UseSphereOrientation=false
   )
   &AddVelocity(Velocity=&Multiply_VectorByFloat(Vector=$Output.SphereLocation.SphereNormal, Float=5))
   &CameraOffset(Camera Offset Amount=-3)
  ]
  particleUpdateScript [
   &ParticleState()
   &VectorNoiseForce(
    Force Amount=&MaskFloatBySpawnGroup(Masked Float=200, Passthrough Float=100)
    Module.Use Random Noise Time Interval=false
   )
   &SolveForcesAndVelocity(Clamp Velocity=false, Limit Acceleration=false)
   &ScaleSpriteSize(Uniform Curve Sprite Scale=DataInterface() {propertyValues {Curve {keys [
        <at 0 1 RCIM_Cubic>
        {arriveTangentWeight=0.0462, interpMode=RCIM_Cubic, leaveTangentWeight=0.0449, tangentMode=RCTM_User, time=0.1354, value=0.2886}
        {arriveTangentWeight=0.0512, interpMode=RCIM_Cubic, leaveTangentWeight=0.0563, tangentMode=RCTM_User, time=0.2671, value=0.9417}
        {arriveTangentWeight=0.0484, interpMode=RCIM_Cubic, leaveTangentWeight=0.0536, tangentMode=RCTM_User, time=0.412, value=0.1231}
        {arriveTangentWeight=0.0535, interpMode=RCIM_Cubic, leaveTangentWeight=0.0475, tangentMode=RCTM_User, time=0.5725, value=0.9591}
        {arriveTangentWeight=0.0479, interpMode=RCIM_Cubic, leaveTangentWeight=0.0524, tangentMode=RCTM_User, time=0.7149, value=0.1294}
        {arriveTangentWeight=0.0536, interpMode=RCIM_Cubic, leaveTangentWeight=0.0443, tangentMode=RCTM_User, time=0.8709, value=0.9199}
        {arriveTangentWeight=0.043, interpMode=RCIM_Cubic, tangentMode=RCTM_User, time=1, value=0.2}
       ]}}})
   &ScaleColor(
    Scale Alpha=&FloatFromCurve(FloatCurve=DataInterface() {propertyValues {Curve {keys [
         {interpMode=RCIM_Cubic, leaveTangentWeight=0.3333, tangentMode=RCTM_User, time=0, value=1}
         {arriveTangentWeight=0.3333, interpMode=RCIM_Cubic, tangentMode=RCTM_User, time=1, value=0}
        ]}}})
    Scale RGB=&VectorFromFloat(Value=7)
    ScaleA=true
    ScaleRGB=true
   )
  ]
  rendererClasses [&NiagaraSpriteRendererProperties]
  renderers [{rendererClass=&NiagaraSpriteRendererProperties}]
 }
 Horn_unOpt {
  bEnabled=false
  simTarget=GPUComputeSim
  emitterUpdateScript [
   &EmitterState()
   &SpawnRate(
    SpawnRate=&MakeCustomFloatFromBool(Bool=$User.bIsFullyDeployed, True Float=10, False Float=0)
    Use Spawn Probability=false
   )
  ]
  particleSpawnScript [
   &InitializeParticle(
    Color=<rgba 1 1 1 1.5>
    Lifetime Max=1.25
    Lifetime Min=0.85
    Lifetime Mode=::NewEnumerator1
    Sprite Rotation Mode=::NewEnumerator2
    Sprite Size Mode=::NewEnumerator1
    Uniform Sprite Size Max=7
    Uniform Sprite Size Min=1
    UsePositionOffset=false
    Write Material Random=true
   )
   &SkeletalMeshLocation(
    MaskSpawnGroup=false
    Mesh Sampling Type=::NewEnumerator5
    Sampled Position Offset=<20 0 0>
    Skeletal Mesh=$User.Glider
    Socket Sampling Mode=::NewEnumerator5
   )
   &ShapeLocation(
    Additional Quaternion Rotation=true
    Additional Yaw / Pitch / Roll=<0.5 0 0>
    Cone Angle=2
    MaskSpawnGroup=false
    Override Seed=false
    Rotation Mode=::NewEnumerator3
    Rotation Quaternion=$Particles.SkeletalMeshLocation.SampledRotation
    Shape Primitive=::NewEnumerator5
    UseOwnerScale=true
   )
   &AddVelocity(Velocity=&Multiply_VectorByFloat(Vector=$Particles.ShapeLocation.ShapeNormal, Float=&RandomRangeFloat(Minimum=2, Maximum=5)))
   &AddVelocity(Velocity=&RotateVectorByQuaternion(
     Vector To Rotate=&MakeVector(X=&RandomRangeFloat(Minimum=5, Maximum=15), Y=0, Z=0)
     Quaternion=$Particles.SkeletalMeshLocation.SampledRotation
    )) {module=AddVelocity001}
   &CameraOffset(Camera Offset Amount=5)
  ]
  particleUpdateScript [
   &ParticleState()
   &AccelerationForce(Acceleration=&MakeVector(X=0, Y=0, Z=&RandomRangeFloat(Minimum=5, Maximum=20)))
   &Drag(Drag=&RandomRangeFloat(Minimum=1, Maximum=3), Use Linear Drag=true, Use Rotational Drag=false)
   &SolveForcesAndVelocity(Clamp Velocity=false, Limit Acceleration=false)
   &SpriteRotationRate(Rotation Rate=&RandomRangeFloat(Minimum=-0.1, Maximum=0.12))
   &ScaleColor(
    Scale Alpha=&FloatFromCurve(FloatCurve=DataInterface() {propertyValues {Curve {keys [
         {interpMode=RCIM_Cubic, leaveTangentWeight=0.3333, tangentMode=RCTM_User, time=0, value=1}
         {arriveTangentWeight=0.3333, interpMode=RCIM_Cubic, tangentMode=RCTM_User, time=1, value=0}
        ]}}})
    ScaleA=true
    ScaleRGB=true
   )
   &DynamicMaterialParameters(
    Index 0 Param 1=&FloatFromCurve(FloatCurve=DataInterface() {propertyValues {Curve {keys [
         {interpMode=RCIM_Cubic, leaveTangentWeight=0.15, tangentMode=RCTM_User, time=0, value=1.2}
         {
          arriveTangent=-2.0203
          arriveTangentWeight=0.3808
          interpMode=RCIM_Cubic
          leaveTangent=-2.0203
          leaveTangentWeight=0.1269
          tangentMode=RCTM_User
          time=0.45
          value=0
         }
         {arriveTangentWeight=0.1038, interpMode=RCIM_Cubic, leaveTangentWeight=0.2767, tangentMode=RCTM_User, time=0.6, value=-0.2}
         {arriveTangentWeight=0.1333, interpMode=RCIM_Cubic, tangentMode=RCTM_User, time=1, value=1}
        ]}}})
    Index 0 Param 2=&FloatFromCurve(FloatCurve=DataInterface() {propertyValues {Curve {keys [
         {interpMode=RCIM_Cubic, leaveTangentWeight=0.2833, tangentMode=RCTM_User, time=0, value=3.5}
         {arriveTangentWeight=1.0313, interpMode=RCIM_Cubic, leaveTangentWeight=0.182, tangentMode=RCTM_User, time=0.85, value=2.8}
         {arriveTangentWeight=0.05, interpMode=RCIM_Cubic, tangentMode=RCTM_User, time=1, value=0}
        ]}}})
    Index 0 Param 3=&FloatFromCurve(FloatCurve=DataInterface() {propertyValues {Curve {keys [
         {interpMode=RCIM_Cubic, leaveTangentWeight=0.1, tangentMode=RCTM_User, time=0.2, value=0}
         {arriveTangentWeight=0.1, interpMode=RCIM_Cubic, leaveTangentWeight=0.1267, tangentMode=RCTM_User, time=0.5, value=4}
         {arriveTangentWeight=0.1267, interpMode=RCIM_Cubic, tangentMode=RCTM_User, time=0.88, value=0}
        ]}}})
    Index 0 Param 4=&RandomRangeFloat(Minimum=0, Maximum=4)
    Module.Param0WriteEnabled=true
    Module.Param1WriteEnabled=true
    Module.Param2WriteEnabled=true
    Module.Param3WriteEnabled=true
   )
   &ScaleSpriteSize(Uniform Curve Sprite Scale=DataInterface() {propertyValues {Curve {keys [<at 0 1 RCIM_Cubic>,<at 0.75 1.2 RCIM_Cubic>,<at 1 1 RCIM_Cubic>]}}})
  ]
  rendererClasses [&NiagaraSpriteRendererProperties]
  renderers [{rendererClass=&NiagaraSpriteRendererProperties}]
 }
 Magic {
  simTarget=GPUComputeSim
  emitterUpdateScript [
   &EmitterState()
   &SpawnParticlesFromOtherEmitter(
    Attribute Reader=DataInterface() {propertyValues {EmitterBinding {bindingMode=Other, emitterName=Head}}}
    Spawn Rate=75
    Spawn Rate Per Particle Cap=500
    Use Maximum Age=false
    Use Minimum Age=false
   )
   &SpawnParticlesFromOtherEmitter(
    Attribute Reader=DataInterface() {propertyValues {EmitterBinding {bindingMode=Other, emitterName=Head}}}
    Spawn Group=1
    Spawn Rate=75
    Spawn Rate Per Particle Cap=500
    Use Maximum Age=false
    Use Minimum Age=false
   ) {module=SpawnParticlesFromOtherEmitter001}
  ]
  particleSpawnScript [
   &InitializeParticle(
    Color=&MaskLinearColorBySpawnGroup(Masked Linear Color=<rgba 0.5729 0.8173 1 1>, Passthrough Linear Color=<rgba 0 0.5722 1 1>)
    Lifetime Max=0.5
    Lifetime Min=0.125
    Lifetime Mode=::NewEnumerator1
    Sprite Size Mode=::NewEnumerator1
    Uniform Sprite Size Max=&MaskFloatBySpawnGroup(Masked Float=5, Passthrough Float=10)
    Uniform Sprite Size Min=&MaskFloatBySpawnGroup(Masked Float=1, Passthrough Float=5)
    UsePositionOffset=false
    Write Material Random=true
   )
   &SampleParticlesFromOtherEmitter()
   &SphereLocation(
    MaskSpawnGroup=false
    Override Seed=false
    Sphere Radius=&MaskFloatBySpawnGroup(Masked Float=3, Passthrough Float=0.1)
    Surface Only=false
    UseNonUniformScale=false
    UseOffset=false
    UseSphereOrientation=false
   )
   &AddVelocity(Velocity=&Multiply_VectorByFloat(Vector=$Output.SphereLocation.SphereNormal, Float=20))
   &CameraOffset(Camera Offset Amount=3)
  ]
  particleUpdateScript [
   &ParticleState()
   &VectorNoiseForce(
    Force Amount=&MaskFloatBySpawnGroup(Masked Float=200, Passthrough Float=100)
    Module.Use Random Noise Time Interval=false
   )
   &SolveForcesAndVelocity(Clamp Velocity=false, Limit Acceleration=false)
   &ScaleSpriteSize(Uniform Curve Sprite Scale=DataInterface() {propertyValues {Curve {keys [
        <at 0 1 RCIM_Cubic>
        {arriveTangentWeight=0.0462, interpMode=RCIM_Cubic, leaveTangentWeight=0.0449, tangentMode=RCTM_User, time=0.1354, value=0.2886}
        {arriveTangentWeight=0.0512, interpMode=RCIM_Cubic, leaveTangentWeight=0.0563, tangentMode=RCTM_User, time=0.2671, value=0.9417}
        {arriveTangentWeight=0.0484, interpMode=RCIM_Cubic, leaveTangentWeight=0.0536, tangentMode=RCTM_User, time=0.412, value=0.1231}
        {arriveTangentWeight=0.0535, interpMode=RCIM_Cubic, leaveTangentWeight=0.0475, tangentMode=RCTM_User, time=0.5725, value=0.9591}
        {arriveTangentWeight=0.0479, interpMode=RCIM_Cubic, leaveTangentWeight=0.0524, tangentMode=RCTM_User, time=0.7149, value=0.1294}
        {arriveTangentWeight=0.0536, interpMode=RCIM_Cubic, leaveTangentWeight=0.0443, tangentMode=RCTM_User, time=0.8709, value=0.9199}
        {arriveTangentWeight=0.043, interpMode=RCIM_Cubic, tangentMode=RCTM_User, time=1, value=0.2}
       ]}}})
   &ScaleColor(
    Scale Alpha=&FloatFromCurve(FloatCurve=DataInterface() {propertyValues {Curve {keys [
         {interpMode=RCIM_Cubic, leaveTangentWeight=0.3333, tangentMode=RCTM_User, time=0, value=1}
         {arriveTangentWeight=0.3333, interpMode=RCIM_Cubic, tangentMode=RCTM_User, time=1, value=0}
        ]}}})
    Scale RGB=&VectorFromFloat(Value=7)
    ScaleA=true
    ScaleRGB=true
   )
  ]
  rendererClasses [&NiagaraSpriteRendererProperties]
  renderers [{rendererClass=&NiagaraSpriteRendererProperties}]
 }
 Ribbon_Trails {
  emitterUpdateScript [
   &EmitterState(Life Cycle Mode=::NewEnumerator1, Loop Behavior=::NewEnumerator1, Loop Duration Mode=::NewEnumerator1, UseLoopDelay=false)
   &SpawnBurst_Instantaneous(Spawn Count=20, Use Spawn Probability=false, UseLoopCountLimit=false)
  ]
  particleSpawnScript [
   &FN_SetSpawnGroup(Normalize Spawn Group=true, Number of Spawn Groups=2, Particles Per Group=10, Set Ribbon ID and Ribbon Link Order=true)
   &InitializeRibbon(
    Color=<rgba 0 0.5722 1 0.1>
    Lifetime=0
    Ribbon Facing Mode:ENiagara_UnsetDirectSet=::NewEnumerator1
    Ribbon Width=25
    UsePositionOffset=false
    Write Material Random=true
   )
  ]
  particleUpdateScript [
   &ParticleState(Kill Particles When Lifetime Has Elapsed=false, Let Infinitely Lived Particles Die When Emitter Deactivates=true)
   &SkeletalMeshLocation(
    Bone / Socket Index=&MaskIntBySpawnGroup(Masked Int=2, Passthrough Int=3, Spawn Group Mask=0)
    MaskSpawnGroup=false
    Mesh Sampling Type=::NewEnumerator5
    Sampled Position Offset=&Lerp_Vector(
     A=<-150 0 -5>
     B=/Niagara/DynamicInputs/Lerp/Lerp_Position(A:NiagaraPosition=<-150 0 -5>, B:NiagaraPosition=<250 0 -5>, Alpha:NiagaraFloat=$System.Timeline001.NormalizedTime)
     Alpha=&ReturnNormalizedExecIndex()
    )
    Skeletal Mesh=$User.Glider
    Socket Sampling Mode=::NewEnumerator5
   )
   &ScaleColor(
    Scale Alpha=&FloatFromCurve(
     FloatCurve=DataInterface() {propertyValues {Curve {keys [
         {interpMode=RCIM_Cubic, tangentMode=RCTM_User, time=0, value=1}
         {arriveTangent=-3.0673, interpMode=RCIM_Cubic, leaveTangent=-3.0673, tangentMode=RCTM_User, time=1, value=0}
        ]}}}
     CurveIndex=&ReturnNormalizedExecIndex()
    )
    Scale RGB=&VectorFromFloat(Value=&FloatFromCurve(
      FloatCurve=DataInterface() {propertyValues {Curve {keys [<at 0 5 RCIM_Cubic>, {arriveTangent=-14.3611, interpMode=RCIM_Cubic, leaveTangent=-14.3611, tangentMode=RCTM_User, time=1, value=1}]}}}
      CurveIndex=&ReturnNormalizedExecIndex()
     ))
    ScaleA=true
    ScaleRGB=true
   )
   &DynamicMaterialParameters(
    Index 0 Param 1=&Multiply_Float(
     A=$Engine.Time
     B=&FloatFromCurve(FloatCurve=DataInterface() {propertyValues {Curve {keys [<at 0 0>,<at 1 5>]}}}, CurveIndex=$User.ForwardDot, Scale Curve=-1)
    )
    Index 0 Param 2=1
    Index 0 Param 3=1
    Module.Param0WriteEnabled=true
    Module.Param1WriteEnabled=true
    Module.Param2WriteEnabled=true
    Module.Param3WriteEnabled=true
   )
  ]
  rendererClasses [&NiagaraRibbonRendererProperties]
  renderers [{rendererClass=&NiagaraRibbonRendererProperties}]
 }
 Snow_Interior {
  emitterUpdateScript [
   &EmitterState()
   &SpawnRate(
    SpawnRate=&MakeCustomFloatFromBool(Bool=$User.bIsFullyDeployed, True Float=25, False Float=0)
    Use Spawn Probability=false
   )
   &SpawnRate(
    SpawnGroup=1
    SpawnRate=&MakeCustomFloatFromBool(Bool=$User.bIsFullyDeployed, True Float=25, False Float=0)
    Use Spawn Probability=false
   ) {module=SpawnRate001}
  ]
  particleSpawnScript [
   /SetVariables(Particles.VisibilityTag=&SpawnGroup()) {isSetParameters=true, module=SetVariables_#2}
   &InitializeParticle(
    Color=&MaskLinearColorBySpawnGroup(Masked Linear Color=<rgba 0 0.25 1 1>, Passthrough Linear Color=<rgba 1 1 1 1.5>)
    Lifetime Max=1.35
    Lifetime Min=0.95
    Lifetime Mode=::NewEnumerator1
    Sprite Rotation Mode=::NewEnumerator2
    Sprite Size Mode=::NewEnumerator1
    Uniform Sprite Size Max=&MaskFloatBySpawnGroup(Masked Float=7, Passthrough Float=2)
    Uniform Sprite Size Min=&MaskFloatBySpawnGroup(Masked Float=1.5, Passthrough Float=1)
    UsePositionOffset=false
    Write Material Random=true
   )
   &ShapeLocation(
    Cylinder Height=2
    Cylinder Height Midpoint=0
    Cylinder Radius=50
    MaskSpawnGroup=true
    Offset=<0 -40 140>
    Offset Mode=::NewEnumerator0
    Override Seed=false
    Shape Primitive=::NewEnumerator1
    Surface Only=false
    UseOwnerScale=true
   )
   &ShapeLocation(
    Cylinder Height=2
    Cylinder Height Midpoint=0
    Cylinder Radius=50
    MaskSpawnGroup=true
    Offset=<0 -40 140>
    Offset Mode=::NewEnumerator0
    Override Seed=false
    Shape Primitive=::NewEnumerator1
    Spawn Group Mask=1
    Surface Only=false
    UseOwnerScale=true
   ) {module=ShapeLocation001}
   &AddVelocity(Velocity=&MakeVector(X=0, Y=0, Z=&RandomRangeFloat(Minimum=-25, Maximum=-75)))
  ]
  particleUpdateScript [
   &ParticleState()
   &GravityForce(Gravity=&MakeVector(X=0, Y=0, Z=&RandomRangeFloat(Minimum=-75, Maximum=-150)))
   &Drag(Drag=&RandomRangeFloat(Minimum=1, Maximum=2), Use Linear Drag=true, Use Rotational Drag=false)
   &VectorNoiseForce(
    Force Amount=&Multiply_Float(
     A=&FloatFromCurve(
      FloatCurve=DataInterface() {propertyValues {Curve {keys [
          {arriveTangent=4.0849, interpMode=RCIM_Cubic, leaveTangent=4.0849, leaveTangentWeight=0.15, tangentMode=RCTM_User, time=0, value=0}
          {arriveTangentWeight=0.15, interpMode=RCIM_Cubic, tangentMode=RCTM_User, time=0.45, value=1}
         ]}}}
      Scale Curve=500
     )
     B=1
    )
    Module.Use Random Noise Time Interval=false
   )
   &SolveForcesAndVelocity(Clamp Velocity=false, Limit Acceleration=false)
   &ScaleColor(
    Scale Alpha=&FloatFromCurve(FloatCurve=DataInterface() {propertyValues {Curve {keys [
         {interpMode=RCIM_Cubic, leaveTangentWeight=0.0667, tangentMode=RCTM_User, time=0, value=0}
         {arriveTangentWeight=0.0667, interpMode=RCIM_Cubic, leaveTangentWeight=0.2667, tangentMode=RCTM_User, time=0.2, value=1.2}
         {arriveTangentWeight=0.2667, interpMode=RCIM_Cubic, tangentMode=RCTM_User, time=1, value=0}
        ]}}})
    ScaleA=true
    ScaleRGB=true
   )
   &SpriteRotationRate(Rotation Rate=&RandomRangeFloat(Minimum=-0.4, Maximum=0.4))
   &ScaleSpriteSize(Uniform Curve Sprite Scale=DataInterface() {propertyValues {Curve {keys [
        <at 0 0 RCIM_Cubic>
        {arriveTangent=3.2789, interpMode=RCIM_Cubic, leaveTangent=3.2789, time=0.15, value=1.142}
        {arriveTangent=-1.3435, interpMode=RCIM_Cubic, leaveTangent=-1.3435, time=0.42, value=1.3772}
        <at 1 0 RCIM_Cubic>
       ]}}})
   /SetVariables(
    Particles.Snowflake_B_Color:LinearColor=&MultiplyLinearColor(
     Linear Color=$Particles.Initial.Color
     Scale Factor=&MakeFloat4FromVectorAndFloat(
      "Vector (RGB)"=<1 1 1>
      "Float (Alpha)"=&FloatFromCurve(FloatCurve=DataInterface() {propertyValues {Curve {keys [
           {interpMode=RCIM_Cubic, leaveTangentWeight=0.3333, tangentMode=RCTM_User, time=0, value=1}
           {arriveTangentWeight=0.3333, interpMode=RCIM_Cubic, tangentMode=RCTM_User, time=1, value=0}
          ]}}})
     )
    )
    Particles.Snowflake_B_Size:Vector2f=&Multiply_Vector2D(
     A=$Particles.Initial.SpriteSize
     B=&Vector2DFromFloat(Value=&FloatFromCurve(FloatCurve=DataInterface() {propertyValues {Curve {keys [<at 0 1 RCIM_Cubic>,<at 0.75 1.2 RCIM_Cubic>,<at 1 1 RCIM_Cubic>]}}}))
    )
   ) {isSetParameters=true, module=SetVariables_#3}
   &DynamicMaterialParameters(
    Index 0 Param 1=&FloatFromCurve(FloatCurve=DataInterface() {propertyValues {Curve {keys [
         {interpMode=RCIM_Cubic, leaveTangentWeight=0.15, tangentMode=RCTM_User, time=0, value=1.2}
         {
          arriveTangent=-2.0203
          arriveTangentWeight=0.3808
          interpMode=RCIM_Cubic
          leaveTangent=-2.0203
          leaveTangentWeight=0.1269
          tangentMode=RCTM_User
          time=0.45
          value=0
         }
         {arriveTangentWeight=0.1038, interpMode=RCIM_Cubic, leaveTangentWeight=0.2767, tangentMode=RCTM_User, time=0.6, value=-0.2}
         {arriveTangentWeight=0.1333, interpMode=RCIM_Cubic, tangentMode=RCTM_User, time=1, value=1}
        ]}}})
    Index 0 Param 2=&FloatFromCurve(FloatCurve=DataInterface() {propertyValues {Curve {keys [
         {interpMode=RCIM_Cubic, leaveTangentWeight=0.2833, tangentMode=RCTM_User, time=0, value=3.5}
         {arriveTangentWeight=1.0313, interpMode=RCIM_Cubic, leaveTangentWeight=0.182, tangentMode=RCTM_User, time=0.85, value=2.8}
         {arriveTangentWeight=0.05, interpMode=RCIM_Cubic, tangentMode=RCTM_User, time=1, value=0}
        ]}}})
    Index 0 Param 3=&FloatFromCurve(FloatCurve=DataInterface() {propertyValues {Curve {keys [
         {interpMode=RCIM_Cubic, leaveTangentWeight=0.1, tangentMode=RCTM_User, time=0.2, value=0}
         {arriveTangentWeight=0.1, interpMode=RCIM_Cubic, leaveTangentWeight=0.1267, tangentMode=RCTM_User, time=0.5, value=4}
         {arriveTangentWeight=0.1267, interpMode=RCIM_Cubic, tangentMode=RCTM_User, time=0.88, value=0}
        ]}}})
    Index 0 Param 4=&RandomRangeFloat(Minimum=0, Maximum=4)
    Module.Param0WriteEnabled=true
    Module.Param1WriteEnabled=true
    Module.Param2WriteEnabled=true
    Module.Param3WriteEnabled=true
   )
  ]
  rendererClasses [&NiagaraSpriteRendererProperties]
  renderers [
   {rendererClass=&NiagaraSpriteRendererProperties}
   {rendererClass=&NiagaraSpriteRendererProperties, rendererIndex=1}
  ]
 }
 Snow_Tracks {
  emitterUpdateScript [
   &EmitterState()
   &SpawnRate(
    SpawnRate=&MakeCustomFloatFromBool(Bool=$User.bIsFullyDeployed, True Float=50, False Float=0)
    Use Spawn Probability=false
   )
   &SpawnRate(
    SpawnGroup=1
    SpawnRate=&MakeCustomFloatFromBool(Bool=$User.bIsFullyDeployed, True Float=50, False Float=0)
    Use Spawn Probability=false
   ) {module=SpawnRate001}
  ]
  particleSpawnScript [
   /SetVariables(Particles.VisibilityTag=&SpawnGroup()) {isSetParameters=true, module=SetVariables_#2}
   &InitializeParticle(
    Color=&MaskLinearColorBySpawnGroup(Masked Linear Color=<rgba 0 0.25 1 1>, Passthrough Linear Color=<rgba 1 1 1 1.5>)
    Lifetime Max=1.35
    Lifetime Min=0.95
    Lifetime Mode=::NewEnumerator1
    Sprite Rotation Mode=::NewEnumerator2
    Sprite Size Mode=::NewEnumerator1
    Uniform Sprite Size Max=&MaskFloatBySpawnGroup(Masked Float=7, Passthrough Float=2)
    Uniform Sprite Size Min=&MaskFloatBySpawnGroup(Masked Float=1.5, Passthrough Float=1)
    UsePositionOffset=false
    Write Material Random=true
   )
   &SkeletalMeshLocation(
    Bone / Socket Index=&RandomRangeInt(Minimum=2, Maximum=3)
    MaskSpawnGroup=false
    Mesh Sampling Type=::NewEnumerator5
    Skeletal Mesh=$User.Glider
    Socket Sampling Mode=::NewEnumerator5
   )
   &ShapeLocation(
    Box Midpoint=<0.5 1 0>
    Box Size=<10 150 0>
    MaskSpawnGroup=true
    Offset Mode=::NewEnumerator0
    Override Seed=false
    Shape Primitive=::NewEnumerator2
    Surface Only=false
    UseOwnerScale=true
   )
   &ShapeLocation(
    Box Midpoint=<0.5 1 0>
    Box Size=<10 150 0>
    MaskSpawnGroup=true
    Offset Mode=::NewEnumerator0
    Override Seed=false
    Shape Primitive=::NewEnumerator2
    Spawn Group Mask=1
    Surface Only=false
    UseOwnerScale=true
   ) {module=ShapeLocation001}
   &AddVelocity(Velocity=&MakeVector(
     X=0
     Y=&Multiply_Float(
      A=300
      B=&FloatFromCurve(FloatCurve=DataInterface() {propertyValues {Curve {keys [<at -1 -1>,<at 0 0>,<at 1 1>]}}}, CurveIndex=$User.ForwardDot)
     )
     Z=0
    ))
   &CameraOffset(Camera Offset Amount=5)
  ]
  particleUpdateScript [
   &ParticleState()
   &GravityForce(Gravity=&MakeVector(X=0, Y=0, Z=&RandomRangeFloat(Minimum=-75, Maximum=-150)))
   &Drag(Drag=&RandomRangeFloat(Minimum=1, Maximum=2), Use Linear Drag=true, Use Rotational Drag=false)
   &VectorNoiseForce(
    Force Amount=&Multiply_Float(
     A=&FloatFromCurve(
      FloatCurve=DataInterface() {propertyValues {Curve {keys [
          {arriveTangent=4.0849, interpMode=RCIM_Cubic, leaveTangent=4.0849, leaveTangentWeight=0.15, tangentMode=RCTM_User, time=0, value=0}
          {arriveTangentWeight=0.15, interpMode=RCIM_Cubic, tangentMode=RCTM_User, time=0.45, value=1}
         ]}}}
      Scale Curve=500
     )
     B=1
    )
    Module.Use Random Noise Time Interval=false
   )
   &SolveForcesAndVelocity(Clamp Velocity=false, Limit Acceleration=false)
   &ScaleColor(
    Scale Alpha=&FloatFromCurve(FloatCurve=DataInterface() {propertyValues {Curve {keys [
         {interpMode=RCIM_Cubic, leaveTangentWeight=0.0667, tangentMode=RCTM_User, time=0, value=0}
         {arriveTangentWeight=0.0667, interpMode=RCIM_Cubic, leaveTangentWeight=0.2667, tangentMode=RCTM_User, time=0.2, value=1.2}
         {arriveTangentWeight=0.2667, interpMode=RCIM_Cubic, tangentMode=RCTM_User, time=1, value=0}
        ]}}})
    ScaleA=true
    ScaleRGB=true
   )
   &SpriteRotationRate(Rotation Rate=&RandomRangeFloat(Minimum=-0.4, Maximum=0.4))
   &ScaleSpriteSize(Uniform Curve Sprite Scale=DataInterface() {propertyValues {Curve {keys [
        <at 0 0 RCIM_Cubic>
        {arriveTangent=3.2789, interpMode=RCIM_Cubic, leaveTangent=3.2789, time=0.15, value=1.142}
        {arriveTangent=-1.3435, interpMode=RCIM_Cubic, leaveTangent=-1.3435, time=0.42, value=1.3772}
        <at 1 0 RCIM_Cubic>
       ]}}})
   /SetVariables(
    Particles.Snowflake_B_Color:LinearColor=&MultiplyLinearColor(
     Linear Color=$Particles.Initial.Color
     Scale Factor=&MakeFloat4FromVectorAndFloat(
      "Vector (RGB)"=<1 1 1>
      "Float (Alpha)"=&FloatFromCurve(FloatCurve=DataInterface() {propertyValues {Curve {keys [
           {interpMode=RCIM_Cubic, leaveTangentWeight=0.3333, tangentMode=RCTM_User, time=0, value=1}
           {arriveTangentWeight=0.3333, interpMode=RCIM_Cubic, tangentMode=RCTM_User, time=1, value=0}
          ]}}})
     )
    )
    Particles.Snowflake_B_Size:Vector2f=&Multiply_Vector2D(
     A=$Particles.Initial.SpriteSize
     B=&Vector2DFromFloat(Value=&FloatFromCurve(FloatCurve=DataInterface() {propertyValues {Curve {keys [<at 0 1 RCIM_Cubic>,<at 0.75 1.2 RCIM_Cubic>,<at 1 1 RCIM_Cubic>]}}}))
    )
   ) {isSetParameters=true, module=SetVariables_#3}
   &DynamicMaterialParameters(
    Index 0 Param 1=&FloatFromCurve(FloatCurve=DataInterface() {propertyValues {Curve {keys [
         {interpMode=RCIM_Cubic, leaveTangentWeight=0.15, tangentMode=RCTM_User, time=0, value=1.2}
         {
          arriveTangent=-2.0203
          arriveTangentWeight=0.3808
          interpMode=RCIM_Cubic
          leaveTangent=-2.0203
          leaveTangentWeight=0.1269
          tangentMode=RCTM_User
          time=0.45
          value=0
         }
         {arriveTangentWeight=0.1038, interpMode=RCIM_Cubic, leaveTangentWeight=0.2767, tangentMode=RCTM_User, time=0.6, value=-0.2}
         {arriveTangentWeight=0.1333, interpMode=RCIM_Cubic, tangentMode=RCTM_User, time=1, value=1}
        ]}}})
    Index 0 Param 2=&FloatFromCurve(FloatCurve=DataInterface() {propertyValues {Curve {keys [
         {interpMode=RCIM_Cubic, leaveTangentWeight=0.2833, tangentMode=RCTM_User, time=0, value=3.5}
         {arriveTangentWeight=1.0313, interpMode=RCIM_Cubic, leaveTangentWeight=0.182, tangentMode=RCTM_User, time=0.85, value=2.8}
         {arriveTangentWeight=0.05, interpMode=RCIM_Cubic, tangentMode=RCTM_User, time=1, value=0}
        ]}}})
    Index 0 Param 3=&FloatFromCurve(FloatCurve=DataInterface() {propertyValues {Curve {keys [
         {interpMode=RCIM_Cubic, leaveTangentWeight=0.1, tangentMode=RCTM_User, time=0.2, value=0}
         {arriveTangentWeight=0.1, interpMode=RCIM_Cubic, leaveTangentWeight=0.1267, tangentMode=RCTM_User, time=0.5, value=4}
         {arriveTangentWeight=0.1267, interpMode=RCIM_Cubic, tangentMode=RCTM_User, time=0.88, value=0}
        ]}}})
    Index 0 Param 4=&RandomRangeFloat(Minimum=0, Maximum=4)
    Module.Param0WriteEnabled=true
    Module.Param1WriteEnabled=true
    Module.Param2WriteEnabled=true
    Module.Param3WriteEnabled=true
   )
  ]
  rendererClasses [&NiagaraSpriteRendererProperties]
  renderers [
   {rendererClass=&NiagaraSpriteRendererProperties}
   {rendererClass=&NiagaraSpriteRendererProperties, rendererIndex=1}
  ]
 }
 Snowflakes {
  simTarget=GPUComputeSim
  emitterUpdateScript [
   &EmitterState()
   &SpawnParticlesFromOtherEmitter(
    Attribute Reader=DataInterface() {propertyValues {EmitterBinding {bindingMode=Other, emitterName=Head}}}
    Spawn Rate Per Particle Cap=500
    Use Maximum Age=false
    Use Minimum Age=false
   )
  ]
  particleSpawnScript [
   &InitializeParticle(
    Color=<rgba 1 1 1 1.5>
    Lifetime Max=1.25
    Lifetime Min=0.85
    Lifetime Mode=::NewEnumerator1
    Sprite Size Mode=::NewEnumerator1
    Uniform Sprite Size Max=7
    Uniform Sprite Size Min=1.5
    UsePositionOffset=false
    Write Material Random=true
   )
   &SampleParticlesFromOtherEmitter()
   &SphereLocation(
    MaskSpawnGroup=false
    Override Seed=false
    Sphere Radius=5
    Surface Only=false
    UseNonUniformScale=false
    UseOffset=false
    UseSphereOrientation=false
   )
   &AddVelocity(Velocity=&Multiply_VectorByFloat(Vector=$Output.SphereLocation.SphereNormal, Float=&RandomRangeFloat(Minimum=5, Maximum=10)))
   &CameraOffset(Camera Offset Amount=3)
  ]
  particleUpdateScript [
   &ParticleState()
   &Drag(Drag=&RandomRangeFloat(Minimum=1, Maximum=3), Use Linear Drag=true, Use Rotational Drag=false)
   &SolveForcesAndVelocity(Clamp Velocity=false, Limit Acceleration=false)
   &ScaleSpriteSize(Uniform Curve Sprite Scale=DataInterface() {propertyValues {Curve {keys [
        {interpMode=RCIM_Cubic, leaveTangentWeight=0.25, tangentMode=RCTM_User, time=0, value=1}
        {arriveTangentWeight=0.25, interpMode=RCIM_Cubic, leaveTangentWeight=0.0833, tangentMode=RCTM_User, time=0.75, value=1.2}
        {arriveTangentWeight=0.0833, interpMode=RCIM_Cubic, tangentMode=RCTM_User, time=1, value=1}
       ]}}})
   &ScaleColor(
    Scale Alpha=&FloatFromCurve(FloatCurve=DataInterface() {propertyValues {Curve {keys [
         {interpMode=RCIM_Cubic, leaveTangentWeight=0.3333, tangentMode=RCTM_User, time=0, value=1}
         {arriveTangentWeight=0.3333, interpMode=RCIM_Cubic, tangentMode=RCTM_User, time=1, value=0}
        ]}}})
    Scale RGB=&VectorFromFloat(Value=7)
    ScaleA=true
    ScaleRGB=true
   )
   &DynamicMaterialParameters(
    Index 0 Param 1=&FloatFromCurve(FloatCurve=DataInterface() {propertyValues {Curve {keys [
         {interpMode=RCIM_Cubic, leaveTangentWeight=0.15, tangentMode=RCTM_User, time=0, value=1.2}
         {
          arriveTangent=-2.0203
          arriveTangentWeight=0.3808
          interpMode=RCIM_Cubic
          leaveTangent=-2.0203
          leaveTangentWeight=0.1269
          tangentMode=RCTM_User
          time=0.45
          value=0
         }
         {arriveTangentWeight=0.1038, interpMode=RCIM_Cubic, leaveTangentWeight=0.2767, tangentMode=RCTM_User, time=0.6, value=-0.2}
         {arriveTangentWeight=0.1333, interpMode=RCIM_Cubic, tangentMode=RCTM_User, time=1, value=1}
        ]}}})
    Index 0 Param 2=&FloatFromCurve(FloatCurve=DataInterface() {propertyValues {Curve {keys [
         {interpMode=RCIM_Cubic, leaveTangentWeight=0.2833, tangentMode=RCTM_User, time=0, value=3.5}
         {arriveTangentWeight=1.0313, interpMode=RCIM_Cubic, leaveTangentWeight=0.182, tangentMode=RCTM_User, time=0.85, value=2.8}
         {arriveTangentWeight=0.05, interpMode=RCIM_Cubic, tangentMode=RCTM_User, time=1, value=0}
        ]}}})
    Index 0 Param 3=&FloatFromCurve(FloatCurve=DataInterface() {propertyValues {Curve {keys [
         {interpMode=RCIM_Cubic, leaveTangentWeight=0.1, tangentMode=RCTM_User, time=0.2, value=0}
         {arriveTangentWeight=0.1, interpMode=RCIM_Cubic, leaveTangentWeight=0.1267, tangentMode=RCTM_User, time=0.5, value=4}
         {arriveTangentWeight=0.1267, interpMode=RCIM_Cubic, tangentMode=RCTM_User, time=0.88, value=0}
        ]}}})
    Index 0 Param 4=&RandomRangeFloat(Minimum=0, Maximum=4)
    Module.Param0WriteEnabled=true
    Module.Param1WriteEnabled=true
    Module.Param2WriteEnabled=true
    Module.Param3WriteEnabled=true
   )
  ]
  rendererClasses [&NiagaraSpriteRendererProperties]
  renderers [{rendererClass=&NiagaraSpriteRendererProperties}]
 }
}
summary {
 systemName=NS_Glider_IceMaiden_Magic
 emitters [
  {emitterName=Head, simTarget=GPUComputeSim, rendererClasses [&NiagaraSpriteRendererProperties]}
  {emitterName=Magic, simTarget=GPUComputeSim, rendererClasses [&NiagaraSpriteRendererProperties]}
  {emitterName=Snowflakes, simTarget=GPUComputeSim, rendererClasses [&NiagaraSpriteRendererProperties]}
  {emitterName=Horn, simTarget=GPUComputeSim, rendererClasses [&NiagaraSpriteRendererProperties]}
  {emitterName=Horn_Magic, simTarget=GPUComputeSim, rendererClasses [&NiagaraSpriteRendererProperties]}
  {emitterName=Snow_Interior, rendererClasses [&NiagaraSpriteRendererProperties]}
  {emitterName=Big_Snow, rendererClasses [&NiagaraSpriteRendererProperties]}
  {emitterName=Front_Snow_Clouds, rendererClasses [&NiagaraSpriteRendererProperties]}
  {emitterName=Snow_Tracks, rendererClasses [&NiagaraSpriteRendererProperties]}
  {emitterName=Ribbon_Trails, rendererClasses [&NiagaraRibbonRendererProperties]}
  {bEnabled=false, emitterName=Horn_unOpt, simTarget=GPUComputeSim, rendererClasses [&NiagaraSpriteRendererProperties]}
 ]
 userVariables [
  User.bIsFullyDeployed {type=NiagaraBool}
  User.ForwardDot {type=NiagaraFloat, defaultValue=0, description=""}
  User.Glider {
   type=NiagaraDataInterfaceSkeletalMesh
   defaultValue {
    struct=NiagaraExt_VariableValue_DataInterface
    value {dataInterface=&self:NiagaraDataInterfaceSkeletalMesh_0}
   }
  }
 ]
}
